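# Sound cue for a looping power plant turbine ambience, filed under exterior prop sounds and assigned to the Fort_Ambient_TileMix sound class.
# A Random node picks one of two looping turbine waves with equal weight.
# Overridden attenuation uses a box shape with falloff, a wide stereo spread and occlusion traced on a custom game channel, applying a smoothed low-pass filter.
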SoundCue /Game/Sounds/Fort_Prop_Sounds/Exterior/PowerPlant/Turbine_Loop_Cue
root=Random_0
nodes [
 {id=Random_0, children [WavePlayer_1,WavePlayer_2], properties {weights [1,1]}}
 {id=WavePlayer_1, properties {bLooping=true, soundWaveAssetPtr=&/Turbine_Loop_01}}
 {id=WavePlayer_2, properties {bLooping=true, soundWaveAssetPtr=&/Turbine_Loop_02}}
]
properties {
 bOverrideAttenuation=true
 maxDistance=2102.7395
 soundClassObject=/Game/Sounds/FortSoundClassesAndModes/Fort_Ambient_TileMix
 volumeMultiplier=1
 attenuationOverrides {
  attenuationShape=Box
  bEnableOcclusion=true
  falloffDistance=1302.7395
  occlusionInterpolationTime=2
  occlusionLowPassFilterFrequency=600
  occlusionTraceChannel=ECC_GameTraceChannel11
  reverbWetLevelMin=0.6905
  stereoSpread=5000
  attenuationShapeExtents=<600 600 800>
 }
}
